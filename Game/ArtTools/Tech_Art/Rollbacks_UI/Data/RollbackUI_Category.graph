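# Editor utility widget that serves as a category container for cvar checkboxes in the Rollbacks UI art tool.
# AddNewCvar creates a RollbackUI_cvar_checkBox, using the cvar name for label and cvar string and the description for tooltip, then adds it to the AddCvarsHere vertical box.
# The State and Jiras inputs and the Category variable are unused here, and the event graph's events are empty.
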EditorUtilityWidgetBlueprint /Game/ArtTools/Tech_Art/Rollbacks_UI/Data/RollbackUI_Category
graphs [
 AddNewCvar {nodes [
   n1 = AddNewCvar!() -> (then, Cvar:Name, State:bool=~, Jiras:String, Description:String) {is=title K2Node_FunctionEntry}
   > n2 = Create Widget!(
    execute:Exec
    WorldContextObject:Object Reference
    Class:User Widget Class Reference=&/RollbackUI_cvar_checkBox.RollbackUI_cvar_checkBox_C
    OwningPlayer:Player Controller Object Reference
    labelText:Text="To Text (Name)"!(InName:Name=@n1) -> (ReturnValue:Text)
    labelToolTip:Text="To Text (String)"(InString=@n1.Description)
    cvar:String="To String (Name)"(InName=@n1)
   ) -> (then:Exec, ReturnValue:Rollback UI Cvar Check Box Object Reference)
   > AddChild(self=Get AddCvarsHere() -> (AddCvarsHere:Vertical Box Object Reference), Content=@n2)
  ]}
 EventGraph {nodes [
   Event PreConstruct()
   Event Construct()
   Event Tick()
  ]}
]
variables [Category {type=string, default=False}]
widgets {
 info {rootWidgetClass=&Overlay, widgetCount=5}
 widgets [
  Overlay_16 {widgetClassPath=&Overlay, widget=Overlay_16}
  HorizontalBox_0 {
   widgetClassPath=&HorizontalBox
   parent=Overlay_16
   widget=HorizontalBox_0
   slot {class=OverlaySlot, properties {verticalAlignment=VAlign_Fill}}
  }
  Spacer_191 {widgetClassPath=&Spacer, parent=HorizontalBox_0, widget=Spacer_191, properties {size=<5 1>}, slot {class=HorizontalBoxSlot}}
  AddCvarsHere {widgetClassPath=&VerticalBox, bIsVariable=true, parent=HorizontalBox_0, widget=AddCvarsHere, slot {class=HorizontalBoxSlot}}
  Spacer_244 {widgetClassPath=&Spacer, parent=HorizontalBox_0, widget=Spacer_244, properties {size=<5 1>}, slot {class=HorizontalBoxSlot}}
 ]
}
